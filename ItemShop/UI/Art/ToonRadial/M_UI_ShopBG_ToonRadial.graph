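# UI material for an item shop background, drawing slowly rotating toon-style radial rays over layered blue gradients.
# A UVSwitch parameter blends screen-space and regular scale-to-fit UVs, which feed named reroutes for offset center, radial and linear gradients, and a ray mask.
# A time-driven rotator feeds a segmented-triangle function, clipped by Min and combined with a gradient by Max.
# The result drives nested blue color lerps, finished with slight desaturation.
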Material /ItemShop/UI/Art/ToonRadial/M_UI_ShopBG_ToonRadial
graph {
 properties {bAutomaticallySetUsageInEditor=false, usedShadingModels=Default Lit}
 nodes [
  n1 = LinearInterpolate(
   A=Constant3Vector() {constant=<rgba 0.0116 0.107 0.4397 1>}
   B=Constant3Vector() {constant=<rgba 0.0477 0.2697 0.8333 1>}
   Alpha=NamedRerouteUsage()
  )
  n2 = NamedRerouteUsage()
  n3 = Rotator(
   Coordinate=NamedRerouteUsage()
   Time=Multiply(A=Time(), B=ScalarParameter() {defaultValue=0.2, group=Rays, parameterName=Rotation Speed}) {constB=0.2}
  ) {speed=0.2}
  n4 = ScalarParameter() {defaultValue=12, group=Rays, parameterName=Segment Amount}
  n5 = ScalarParameter() {defaultValue=0.6, group=Rays, parameterName=Spacing}
  n6 = Max(
   A=NamedRerouteUsage()
   B=Min(
    A=Multiply(A=@n2)
    B=MaterialFunctionCall!(UV=@n3, Segment Count=@n4, Segment Spacing=@n5, Fixed Spacing) {materialFunction=/FortUILibrary/MaterialFunctions/Basics/MF_UI_RadialSegments}.Segmented Triangles Fill
   )
  )
  n7 = LinearInterpolate(A=@n1, B=Constant3Vector() {constant=<rgba 0.1119 0.5395 0.956 1>}, Alpha=@n6)
  n8 = Multiply(A=@n6) {constB=0.65}
  Desaturation(
   None=LinearInterpolate(A=LinearInterpolate(A=@n7, B=@n1, Alpha=@n8), B=Constant3Vector() {constant=<rgba 0.0044 0.0262 0.1221 1>}, Alpha=NamedRerouteUsage())
   Fraction=Constant() {r=0.25}
  )
  n9 = NamedRerouteUsage()
  n10 = ScalarParameter() {defaultValue=0.5, group=FrontGradient, parameterName=Size}
  NamedRerouteDeclaration(None=MaterialFunctionCall(UV=MaterialFunctionCall(UVs=@n9, Texture Scale=@n10) {materialFunction=&ScaleUVsByCenter}.UVs) {materialFunction=&MF_UI_RadialGradient}.Smooth) {name=FrontGradient, nodeColor=<rgba 0 1 0.8125 1>}
  n11 = LinearInterpolate(
   A=MaterialFunctionCall!(ScreenUV, TextureSize, ScalePivotPoint) {materialFunction=/FortUILibrary/MaterialFunctions/Basics/MF_UI_ScreenSpaceUVScaling}.Scaled to Fit
   B=MaterialFunctionCall() {materialFunction=&MF_UI_UVScaling}.Scaled to Fit
   Alpha=ScalarParameter() {parameterName=UVSwitch}
  )
  n12 = ScalarParameter() {defaultValue=-0.55, group=FrontGradient, parameterName=X Axis}
  n13 = ScalarParameter() {defaultValue=0.15, group=FrontGradient, parameterName=Y Axis}
  NamedRerouteDeclaration(None=Add(A=NamedRerouteDeclaration(None=@n11) {name=ScaletoFitUV, nodeColor=<rgba 1 0 0.8672 1>}, B=AppendVector(A=@n12, B=@n13))) {name=CenterPointUV, nodeColor=<rgba 0 0.1562 1 1>}
  n14 = MaterialFunctionCall(UVs=NamedRerouteUsage(), Texture Scale=ScalarParameter() {defaultValue=1.25, group=BG Gradient, parameterName=BGSize}) {materialFunction=&ScaleUVsByCenter}
  NamedRerouteDeclaration(None=MaterialFunctionCall!("Value (0-1)"=MaterialFunctionCall(UV=@n14.UVs) {materialFunction=&MF_UI_RadialGradient}.Smooth) {materialFunction=/FortUILibrary/MaterialFunctions/Basics/MF_UI_EaseCurves_02_Square}.In-Out) {name=BGGradient, nodeColor=<rgba 0 1 0.25 1>}
  n15 = NamedRerouteUsage()
  n16 = ScalarParameter() {defaultValue=0.6, group=LinearGradient, parameterName=GradientCutoff}
  n17 = ScalarParameter() {defaultValue=-0.4, group=LinearGradient, parameterName=GradientOffset}
  NamedRerouteDeclaration(None=OneMinus(None=MaterialFunctionCall(UVs=@n15, GradientEnd=@n16, GradientOffset=@n17) {materialFunction=&MF_UI_LinearGradient}.Result)) {name=linearGradient, nodeColor=<rgba 0.9688 0 1 1>}
  n18 = Constant() {r=0.05}
  n19 = Constant()
  n20 = Constant() {r=2.4}
  n21 = Constant() {r=0.75}
  n22 = MaterialFunctionCall(UVs=Add(A=NamedRerouteUsage(), B=AppendVector(A=@n18, B=@n19)), Texture Scale=Multiply(A=AppendVector(A=@n20, B=@n21)) {constB=1.65}) {materialFunction=&ScaleUVsByCenter}
  NamedRerouteDeclaration(None=MaterialFunctionCall!("Value (0-1)"=MaterialFunctionCall(UV=@n22.UVs) {materialFunction=&MF_UI_RadialGradient}.Smooth) {materialFunction=/FortUILibrary/MaterialFunctions/Basics/MF_UI_EaseCurves_01_Sine}.In-Out) {name=MaskforRAys, nodeColor=<rgba 0.1016 1 0 1>}
 ]
}
